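# Niagara system for the base visual effect of the LimeRock countdown timer, tinted and sized through User.MasterColorAndOpacity and User.MasterScale.
# Main burst-spawns on a four-second loop and feeds five mesh and two sprite renderers, deriving base, cone, pulse and glow colors from particle color.
# Nested waveforms flicker sprite size, and a curve animates a pulse parameter.
# Rays continuously spawns short-lived, upward-biased meshes that slowly spin and fade in and out.
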NiagaraSystem /LimeRock_CountdownTimer/Gameplay/CountdownTimer/FX/Timer/NS_LimeRock_TimerBase
dependencies {usedDataInterfaces [&NiagaraDataInterfaceCurve], usedDynamicInputs [], usedModules [&SystemState]}
emitters {
 Main {
  emitterUpdateScript [
   &EmitterState(Life Cycle Mode=::NewEnumerator1, Loop Duration=4, UseLoopDelay=false)
   &SpawnBurst_Instantaneous(Use Spawn Probability=false, UseLoopCountLimit=false)
  ]
  particleSpawnScript [
   &InitializeParticle(
    Lifetime=$Emitter.CurrentLoopDuration
    Mesh Scale=&Multiply_Vector(A=<4.5 7 7>, B=$User.MasterScale)
    Mesh Scale Mode=::NewEnumerator4
    Sprite Size Mode=::NewEnumerator3
    Uniform Sprite Size=400
    UsePositionOffset=false
    Write Material Random=true
   )
   &InitialMeshOrientation(Mesh Orientation Mode=::NewEnumerator7, Rotation=<0 -0.25 0>, Use Rotation Vector=true)
   /SetVariables(Particles.SpriteCamera Offset=50) {isSetParameters=true}
  ]
  particleUpdateScript [
   &ParticleState()
   &ScaleMeshSize(Scale Factor=&Multiply_Vector(A=<1 1 1>, B=$User.MasterScale))
   &ScaleSpriteSize(
    Scale Sprite Size Mode=::NewEnumerator0
    Uniform Scale Factor=&Multiply_Float(
     A=&Waveform(
      "Waveform [1]"=::NewEnumerator12
      "[1] Phase Scale"=&Waveform("Waveform [1]"=::NewEnumerator12, "[1] Input Position"=16789.3105, "[1] Frequency"=1, "[1] Amplitude Min/Max"=<0.05 1>)
      "[1] Frequency"=50
      "[1] Amplitude Min/Max"=<0.85 1>
     )
     B=1
    )
   )
   &ScaleColor(
    Scale Mode=::NewEnumerator1
    Scale RGBA=&MakeVector4_FromLinearColor(LinearColor=$User.MasterColorAndOpacity)
    ScaleRGBA=true
   )
   &ScaleColor(ScaleA=true, ScaleRGB=true) {module=ScaleColor001}
   /SetVariables(Particles.Base Color:LinearColor=&MultiplyLinearColor(
     Linear Color=$Particles.Color
     Scale Factor=&MakeFloat4FromVectorAndFloat("Vector (RGB)"=&VectorFromFloat(Value=0.75), "Float (Alpha)"=0.5)
    )) {isSetParameters=true}
   /SetVariables(Particles.Cone_Ray_Dark:LinearColor=&MultiplyLinearColor(
     Linear Color=$Particles.Color
     Scale Factor=&MakeFloat4FromVectorAndFloat("Vector (RGB)"=&VectorFromFloat(Value=0.35), "Float (Alpha)"=1)
    )) {isSetParameters=true}
   /SetVariables(Particles.Cone_Ray_Brightness_Color:LinearColor=&MultiplyLinearColor(
     Linear Color=$Particles.Color
     Scale Factor=&MakeFloat4FromVectorAndFloat("Vector (RGB)"=&VectorFromFloat(Value=1.5), "Float (Alpha)"=1)
    )) {isSetParameters=true}
   /SetVariables(
    Particles.Pulse Color:LinearColor=&MultiplyLinearColor(
     Linear Color=$Particles.Color
     Scale Factor=&MakeFloat4FromVectorAndFloat("Vector (RGB)"=&VectorFromFloat(Value=50), "Float (Alpha)"=1)
    )
    Particles.PusleDynam0:Vector4f=&MakeVector4(
     X=&FloatFromCurve(FloatCurve=DataInterface() {propertyValues {Curve {keys [
          {arriveTangent=2.0695, interpMode=RCIM_Cubic, time=0, value=-1}
          {arriveTangent=0.7391, interpMode=RCIM_Cubic, leaveTangent=0.7391, tangentMode=RCTM_User, time=1, value=1}
         ]}}})
     Y=0
     Z=0
     W=0
    )
   ) {isSetParameters=true}
   /SetVariables(Particles.Soft Color:LinearColor=&MultiplyLinearColor(
     Linear Color=$Particles.Color
     Scale Factor=&MakeFloat4FromVectorAndFloat("Vector (RGB)"=&VectorFromFloat(Value=1), "Float (Alpha)"=1)
    )) {isSetParameters=true}
   /SetVariables(Particles.Base glowColor:LinearColor=&MultiplyLinearColor(
     Linear Color=$Particles.Color
     Scale Factor=&MakeFloat4FromVectorAndFloat("Vector (RGB)"=&VectorFromFloat(Value=7), "Float (Alpha)"=2)
    )) {isSetParameters=true}
   /SetVariables(Particles.OffsetPosition:NiagaraPosition=&AddVectorToPosition(Position=$Particles.Position, Vector=<0 0 30>)) {isSetParameters=true}
   &SpriteFacingAndAlignment(Sprite Facing=<0 0 1>, WriteAlignment=true, WriteFacing=true)
   /SetVariables(
    Particles.Light:LinearColor=<rgba 10 7.5 5 1>
    Particles.Light size:Vector2f=<71.3488 171.0136>
    Particles.LightPosition:NiagaraPosition=&AddVectorToPosition(Position=/Niagara/DynamicInputs/Helpers/SimulationPosition(), Vector=<0 0 12>)
   ) {enabled=false, isSetParameters=true}
  ]
  rendererClasses [&NiagaraMeshRendererProperties, &NiagaraSpriteRendererProperties]
  renderers [
   {rendererClass=&NiagaraMeshRendererProperties, rendererIndex=0}
   {rendererClass=&NiagaraMeshRendererProperties, rendererIndex=1}
   {rendererClass=&NiagaraMeshRendererProperties, rendererIndex=2}
   {rendererClass=&NiagaraMeshRendererProperties, rendererIndex=3}
   {rendererClass=&NiagaraMeshRendererProperties, rendererIndex=4}
   {rendererClass=&NiagaraSpriteRendererProperties, rendererIndex=5}
   {rendererClass=&NiagaraSpriteRendererProperties, rendererIndex=6}
  ]
 }
 Rays {
  emitterUpdateScript [
   &EmitterState()
   &SpawnRate(SpawnRate=8, Use Spawn Probability=false)
  ]
  particleSpawnScript [
   &InitializeParticle(
    Lifetime Max=0.7
    Lifetime Mode=::NewEnumerator1
    Mesh Scale Max=<0.4 0.15 0.1>
    Mesh Scale Min=<0.45 0.2 0.05>
    Mesh Scale Mode=::NewEnumerator2
    UsePositionOffset=false
    Write Material Random=true
   )
   &InitialMeshOrientation(
    Mesh Orientation Mode=::NewEnumerator2
    Orientation Vector=&Lerp_Vector(
     A=&RandomVector(Vector Scale=1, Constrain to Hemisphere=true, Hemisphere Constraint Vector=<0 0 1>)
     B=<0 0 1>
     Alpha=0.45
    )
    Rotation=&RandomRangeVector(Minimum=<0 0 0>, Maximum=<0 0 0>)
    Use Rotation Vector=true
   ) {module=InitialMeshOrientation002}
  ]
  particleUpdateScript [
   &ParticleState()
   &ScaleMeshSize(Scale Factor=&Multiply_Vector(A=&VectorFromFloat(Value=3), B=$User.MasterScale))
   &UpdateMeshOrientation(
    Rotation Coordinate Space=::NewEnumerator1
    Rotation Rate=&UniformAOrBFloat(A=0.1, B=-0.1, Evaluation Type=::NewEnumerator0, Recalculate Random Each Loop=true, Override Seed=false)
    Rotation Vector=<0 0 1>
   )
   &ScaleColor(
    Scale Alpha=&Multiply_Float(
     A=&FloatFromCurve(
      FloatCurve=DataInterface() {propertyValues {Curve {keys [
          {interpMode=RCIM_Cubic, tangentMode=RCTM_User, time=0, value=0}
          {interpMode=RCIM_Cubic, tangentMode=RCTM_User, time=0.3, value=0.5}
          {tangentMode=RCTM_User, time=1, value=0}
         ]}}}
      Scale Curve=0.2
     )
     B=&RandomRangeFloat(Minimum=0.5, Maximum=1.7)
    )
    ScaleA=true
    ScaleRGB=true
   )
   &ScaleColor(
    Scale Mode=::NewEnumerator1
    Scale RGBA=&MakeVector4_FromLinearColor(LinearColor=&MultiplyLinearColor(
      Linear Color=$User.MasterColorAndOpacity
      Scale Factor=&MakeFloat4FromVectorAndFloat("Vector (RGB)"=<1 1 1>, "Float (Alpha)"=1)
     ))
    ScaleRGBA=true
   ) {module=ScaleColor001}
  ]
  rendererClasses [&NiagaraMeshRendererProperties]
  renderers [{rendererClass=&NiagaraMeshRendererProperties, rendererIndex=0}]
 }
}
summary {
 systemName=NS_LimeRock_TimerBase
 emitters [
  {emitterName=Main, rendererClasses [&NiagaraMeshRendererProperties, &NiagaraSpriteRendererProperties]}
  {emitterName=Rays, rendererClasses [&NiagaraMeshRendererProperties]}
 ]
 userVariables [
  User.MasterColorAndOpacity {type=LinearColor, description="", defaultValue {struct=LinearColor, value=<rgba 1 1 0.5 0.075>}}
  User.MasterScale {type=Vector3f, description="", defaultValue {struct=Vector3f, value=<1 1 1>}}
 ]
}
